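# Translucent unlit Niagara material, named as the idle trail effect of the female Eternity pickaxe, drawing a flame-shaped trail with optional nebula coloring.
# Panning distorted noise and UV gradients build the flame mask; a static switch provides a pickaxe swing mask, alongside camera and view-angle fades.
# Emissive blends color parameters with screen-aligned panning space textures and an optional curve-atlas remap.
# Dynamic parameters drive offsets, dissolution panning, wave WPO and interpolated bend.
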Material /BRCosmetics/Weapons/FORT_Melee/PIckaxe_Eternity_Female/FX/M_PIckaxe_Eternity_Female_idle_trail
graph {
 properties {
  bUsedWithNiagaraMeshParticles=true
  bUsedWithNiagaraRibbons=true
  bUsedWithNiagaraSprites=true
  bUsedWithSkeletalMesh=true
  blendMode=BLEND_Translucent
  pixelDepthOffsetMode=PDOM_Legacy
  shadingModel=MSM_Unlit
  twoSided=true
  usedShadingModels=Unlit
 }
 nodes [
  Add(A=NamedRerouteUsage(), B=NamedRerouteUsage())
  n1 = NamedRerouteUsage()
  n2 = Multiply(
   A=Power(Base=NamedRerouteUsage(), Exp=ScalarParameter() {defaultValue=1, group=Color, parameterName=Hot Spot Mask Exp, sortPriority=6})
   B=ScalarParameter() {defaultValue=1, group=Color, parameterName=Hot Spot Mask Multiply, sortPriority=5}
  )
  n3 = Multiply(
   A=LinearInterpolate(
    A=VectorParameter() {group=Color, parameterName=Hot Spot Mask Low Color, sortPriority=3, defaultValue=<rgba 1 1 1 1>}.RGB
    B=VectorParameter() {group=Color, parameterName=Hot Spot Mask High Color, sortPriority=4, defaultValue=<rgba 1 1 1 1>}.RGB
    Alpha=Clamp(None=@n2)
   )
   B=NamedRerouteUsage()
  )
  n4 = ScalarParameter() {group=Color, parameterName=Use Mask Hot Spot, sortPriority=2}
  n5 = MaterialFunctionCall!(TextureResolution=ScalarParameter() {defaultValue=256, group=Texture, parameterName=NebulaTextureResolution, sortPriority=9}) {materialFunction=/Engine/Functions/Engine_MaterialFunctions02/Texturing/ScreenAlignedPixelToPixelUVs}
  n6 = Multiply(A=ComponentMask(None=NamedRerouteUsage()) {g=true, r=true}) {constB=0.05}
  n7 = VectorParameter() {group=Texture, parameterName="Nebula&&StarfieldPannerSpeed", sortPriority=8, defaultValue=<rgba 0 0.05 0 0>}
  n8 = ComponentMask(None=@n7.RGB) {g=true, r=true}
  n9 = TextureSampleParameter2D!(UVs=Panner(Coordinate=Add(A=@n5.Result, B=@n6), Speed=@n8), Level=Constant() {r=1}, Apply View MipBias) {
   group=Texture
   mipValueMode=TMVM_MipLevel
   parameterName=Space_Tex_B
   sortPriority=6
   texture=/Game/Effects/Fort_Effects/Textures/Plasma/T_Plasma_Aurora_Seamless
  }
  n10 = TextureSampleParameter2D(UVs=Panner(Coordinate=@n5.Result, Speed=@n8)) {
   group=Texture
   mipValueMode=TMVM_MipLevel
   parameterName=Space_Tex_A
   sortPriority=5
   texture=/Game/Characters/Player/Male/Medium/Bodies/M_Med_Soldier_04/Skins/Galaxy/Textures/T_FN_Nebula
  }
  n11 = Desaturation(None=@n10.RGB, Fraction=ScalarParameter() {defaultValue=-0.5, group=Texture, parameterName=Space_Tex_A_desaturation, sortPriority=7})
  n12 = Add(A=Desaturation(None=@n9.RGB, Fraction=Constant() {r=-1}), B=@n11)
  n13 = ScalarParameter() {group=Texture, parameterName=Only_Use_one_Space_Tex, sortPriority=4}
  n14 = LinearInterpolate(
   A=Multiply(
    A=LinearInterpolate(A=@n12, B=@n11, Alpha=@n13)
    B=ScalarParameter() {defaultValue=1, group=Texture, parameterName=Texture_intensity, sortPriority=3}
   )
   B=Constant()
   Alpha=ScalarParameter() {group=Texture, parameterName=Use simple texture, sortPriority=2}
  )
  n15 = OneMinus(None=ComponentMask(None=TextureCoordinate()) {g=true})
  n16 = Saturate(None=Power(
    Base=Add(A=@n15) {constB=-0.2}
    Exp=ScalarParameter() {defaultValue=3, group=Texture, parameterName=Space_Texture_center_mask_power, sortPriority=1}
   ) {constExponent=0.5})
  n17 = TextureCoordinate()
  n18 = Constant2Vector() {g=0.5, r=0.5}
  n19 = Multiply(
   A=LinearInterpolate(A=LinearInterpolate(A=@n1, B=@n3, Alpha=@n4), B=Multiply(A=@n14, B=@n16), Alpha=NamedRerouteUsage())
   B=Power(
    Base=SphereMask(A=@n17, B=@n18) {attenuationRadius=0.8, hardnessPercent=0}
    Exp=ScalarParameter() {defaultValue=3, group=Texture, parameterName=Space_Texture_power, sortPriority=0}
   ) {constExponent=3}
  )
  n20 = LinearInterpolate(A=NamedRerouteUsage(), B=Constant() {r=1}, Alpha=ScalarParameter() {group=Global Gradient, parameterName=Not_Use_Global_Gradient})
  n21 = Multiply(A=Multiply(A=Multiply(A=@n19, B=@n20), B=ParticleColor().RGB), B=NamedRerouteUsage())
  n22 = Desaturation(
   None=@n21
   Fraction=ScalarParameter() {defaultValue=1, group=Color Curve, parameterName=Desaturation Amount Before Color Curve, sortPriority=1}
  )
  n23 = NamedRerouteUsage()
  MaterialFunctionCall(Emissive=LinearInterpolate(A=@n21, B=Multiply(A=@n22, B=@n23), Alpha=ScalarParameter() {group=Color Curve, parameterName=UseColorCurve, sortPriority=0})) {materialFunction=&MF_FX_EmissiveAdjustments}
  n24 = StaticSwitchParameter(True=NamedRerouteUsage(), False=NamedRerouteUsage()) {group=Pickaxe Swing Mask, parameterName=Pickaxe_Swing_Mask, sortPriority=0}
  n25 = ScalarParameter() {defaultValue=150, group=Camera Fade, parameterName=Near Camera Distance, sortPriority=1}
  n26 = LinearInterpolate(
   A=@n24
   B=Multiply(A=@n24, B=MaterialFunctionCall(Distance=@n25) {materialFunction=&Near_Camera_Fade}.Result)
   Alpha=ScalarParameter() {group=Camera Fade, parameterName=Use Camera Fade, sortPriority=0}
  )
  n27 = ComponentMask(None=TextureCoordinate()) {g=true}
  n28 = Power(
   Base=@n26
   Exp=LinearInterpolate(
    A=ScalarParameter() {defaultValue=2, group=Ancient Gladiator, parameterName=Final Opacity Power A, sortPriority=1}
    B=ScalarParameter() {defaultValue=1, group=Ancient Gladiator, parameterName=Final Opacity Power B, sortPriority=2}
    Alpha=@n27
   )
  )
  n29 = Clamp(None=Multiply(A=@n27, B=ScalarParameter() {defaultValue=3, group=Ancient Gladiator, parameterName=Final Opacity Gradient Hardness, sortPriority=3}))
  Multiply(
   A=LinearInterpolate(
    A=@n26
    B=Multiply(A=@n28, B=@n29)
    Alpha=ScalarParameter() {group=Ancient Gladiator, parameterName=Use Ancient Gladiator Mask, sortPriority=0}
   )
   B=MaterialFunctionCall("FadeSensitivity (optional)"=ScalarParameter() {group=Camera Fade, parameterName=FadeBasedOnViewAngle, sortPriority=2}) {materialFunction=&FadeBasedOnViewAngle}.Result
  )
  n30 = TextureCoordinate()
  n31 = VectorParameter() {group=Color Curve, parameterName="Color Curve UV Distortion Panner Speed (RG)", sortPriority=10, defaultValue=<rgba 0 0.5 0 1>}
  n32 = TextureSampleParameter2D!(UVs=Panner(Speed=ComponentMask(None=@n31.RGB) {g=true, r=true}) {bFractionalPart=true}, Apply View MipBias) {
   group=Color Curve
   parameterName=Color Curve UV Distortion Texture
   sortPriority=9
   texture=/Game/Global/Textures/Default/Noise/T_FireNoise_01
  }
  n33 = Multiply(A=ScalarParameter() {defaultValue=1, group=Color Curve, parameterName=Color Curve Amount of Distortion, sortPriority=8}) {constB=0.1}
  n34 = VectorParameter() {group=Color Curve, parameterName="Color Curve UV Texture Offset (RG)", sortPriority=7, defaultValue=<rgba 0 -0.25 0 1>}
  n35 = TextureSampleParameter2D(UVs=Add(A=LinearInterpolate(A=@n30, B=@n32.R, Alpha=@n33), B=ComponentMask(None=@n34.RGB) {g=true, r=true})) {
   group=Color Curve
   parameterName=Color Curve UV Texture
   sortPriority=5
   texture=/Game/Accessories/FORT_Backpacks/Backpack_F_MED_Ashes/FX/T_FX_Backpack_Ashes
  }
  n36 = LinearInterpolate(A=@n35.R, B=OneMinus(None=@n35.R), Alpha=ScalarParameter() {group=Color Curve, parameterName=Invert Color Curve UV, sortPriority=6})
  n37 = ScalarParameter() {group=Color Curve, parameterName=Color Curve Texture Inputted UV Contrast, sortPriority=4}
  n38 = CurveAtlasRowParameter(CurveTime=LinearInterpolate(
    A=VertexColor().R
    B=MaterialFunctionCall(In=@n36, Contrast=@n37) {materialFunction=&CheapContrast}.Result
    Alpha=ScalarParameter() {defaultValue=0.55, group=Color Curve, parameterName=DON'T Use Vert Color in Color Curve UV, sortPriority=3}
   )) {
   atlas=/Game/Effects/Niagara/MasterMaterials/Smoke/Textures/CurveAtlas
   curve=/Game/Effects/Niagara/MasterMaterials/Smoke/Textures/EmissiveCurve
   defaultValue=1
   group=Color Curve
   parameterName=ColorCurve
   sortPriority=2
  }
  NamedRerouteDeclaration(None=@n38) {name=Color Curve, nodeColor=<rgba 0 0.8594 1 1>}
  n39 = ParticleColor()
  n40 = SphereMask(A=TextureCoordinate(), B=Constant2Vector() {g=0.5, r=0.5}, Radius=Constant() {r=0.5}) {hardnessPercent=70}
  NamedRerouteDeclaration(None=DepthFade(
    Opacity=Multiply(A=@n39.A, B=@n40)
    FadeDistance=ScalarParameter() {defaultValue=50, group=Pickaxe Swing Mask, parameterName=Round_mask_depth_fade, sortPriority=3}
   )) {name=Simple Round Particle Mask, nodeColor=<rgba 1 0.8203 0 1>}
  n41 = TextureCoordinate()
  n42 = VectorParameter() {group=BottomFade, parameterName=Bottom Fade_SphereMask_A_Scale, sortPriority=2, defaultValue=<rgba 1.2 1 0 0>}
  n43 = ComponentMask(None=@n42.RGB) {g=true, r=true}
  n44 = VectorParameter() {group=BottomFade, parameterName=Bottom Fade_SphereMask_B, sortPriority=3, defaultValue=<rgba 0.6 0.05 0 0>}
  n45 = Multiply(A=SphereMask(
    A=Multiply(A=@n41, B=@n43)
    B=ComponentMask(None=@n44.RGB) {g=true, r=true}
    Radius=ScalarParameter() {defaultValue=0.95, group=BottomFade, parameterName=BottomFade_Radius, sortPriority=1}
   ) {attenuationRadius=0.95, hardnessPercent=0}) {constB=15}
  NamedRerouteDeclaration(None=LinearInterpolate(B=Saturate(None=@n45), Alpha=ScalarParameter() {defaultValue=1, group=BottomFade, parameterName=Bottom Fade, sortPriority=0}) {constA=1}) {name=BottomFade, nodeColor=<rgba 0.3828 1 0 1>}
  n46 = VectorParameter() {group=Fire Shape, parameterName=Fire_MacroShape_UVScale, sortPriority=10, defaultValue=<rgba 1 1 0 0>}
  n47 = MaterialFunctionCall(Float2=Multiply(A=TextureCoordinate(), B=ComponentMask(None=@n46.RGB) {g=true, r=true})) {materialFunction=&BreakOutFloat2Components}
  n48 = Sine(None=Multiply(A=@n47.R) {constB=0.45})
  n49 = NamedRerouteUsage()
  NamedRerouteDeclaration(None=Power(Base=OneMinus(None=@n48), Exp=LinearInterpolate(Alpha=@n49) {constA=0.5, constB=1.5}) {constExponent=0.5}) {name=Flame Split Micro Gradient, nodeColor=<rgba 1 0 0.7031 1>}
  n50 = Multiply(A=@n47.R) {constB=0.5}
  n51 = ScalarParameter() {defaultValue=6, group=Fire Shape, parameterName=Edge Fade Value Top, sortPriority=3}
  n52 = ScalarParameter() {defaultValue=2, group=Fire Shape, parameterName=Edge Fade Value Bottom, sortPriority=4}
  n53 = Multiply(A=Power(Base=Sine(None=@n50), Exp=LinearInterpolate(A=@n51, B=@n52, Alpha=@n47.G) {constA=6, constB=2}) {constExponent=4}, B=@n47.G)
  n54 = Multiply(
   A=NamedRerouteUsage()
   B=Multiply(A=NamedRerouteDeclaration(None=@n53) {name=Macro Mask Shape, nodeColor=<rgba 0 0.0625 1 1>}, B=NamedRerouteUsage())
  )
  n55 = LinearInterpolate(Alpha=@n47.G) {constA=2}
  n56 = ScalarParameter() {defaultValue=2, group=Fire Shape, parameterName=Fire Split Contrast, sortPriority=1}
  n57 = Round(None=Add(A=Power(Base=@n47.G, Exp=@n56), B=NamedRerouteUsage()))
  n58 = ScalarParameter() {defaultValue=1, group=Fire Shape, parameterName=Fire Split, sortPriority=0}
  n59 = Multiply(
   A=Multiply(A=Power(Base=@n54, Exp=@n55), B=LinearInterpolate(B=@n57, Alpha=@n58) {constA=1, constB=0})
   B=ScalarParameter() {defaultValue=15, group=Opacity, parameterName=OpacityMaskMultiplier, sortPriority=1}
  ) {constB=3}
  n60 = LinearInterpolate(A=Floor(None=@n59), B=@n59, Alpha=ScalarParameter() {defaultValue=0.066, group=Opacity, parameterName=Mask Floor})
  n61 = DepthFade(Opacity=@n60, FadeDistance=ScalarParameter() {defaultValue=200, group=Opacity, parameterName=DepthFadeDistance})
  n62 = ParticleColor()
  n63 = ScalarParameter() {defaultValue=1, group=Opacity, parameterName=Multiply by Particle Alpha}
  NamedRerouteDeclaration(None=Multiply(A=ShadingPathSwitch(Default=@n60, Deferred=@n61), B=LinearInterpolate(B=@n62.A, Alpha=@n63) {constA=1})) {name=Flame Shape Mask, nodeColor=<rgba 0 1 0.9766 1>}
  n64 = TextureCoordinate() {uTiling=0.7, vTiling=0.8}
  n65 = Multiply(
   A=ComponentMask(None=NamedRerouteUsage()) {g=true, r=true}
   B=ScalarParameter() {group=Dissolution, parameterName=Dissolution Texture UV Distortion Strength, sortPriority=2}
  )
  n66 = DynamicParameter() {paramNames [Offset, Texture_mask_offset, WaveIntensity, Param4]}
  n67 = Time()
  n68 = ScalarParameter() {group=Dissolution, parameterName=Loop_panner, sortPriority=1}
  n69 = VectorParameter() {group=Dissolution, parameterName="Dissolution Texture Panner Speed (RG)", defaultValue=<rgba 0 0.2 0 0>}
  n70 = TextureSampleParameter2D(UVs=Panner(
    Coordinate=Add(A=@n64, B=@n65)
    Time=LinearInterpolate(A=@n66.Texture_mask_offset, B=@n67, Alpha=@n68)
    Speed=ComponentMask(None=@n69.RGB) {g=true, r=true}
   ) {speedY=0.2}) {
   group=Dissolution
   parameterName=dissolution Texture
   sortPriority=0
   texture=&T_Inky_Smoke_Tile
  }
  NamedRerouteDeclaration(None=@n70.R) {name=Dissolution, nodeColor=<rgba 1 0.8438 0 1>}
  NamedRerouteDeclaration(None=LinearInterpolate(
    A=VectorParameter() {group=Color, parameterName=Color1, sortPriority=0, defaultValue=<rgba 0.0015 0.0028 0.0104 1>}.RGB
    B=VectorParameter() {group=Color, parameterName=Color2, sortPriority=1, defaultValue=<rgba 0.2135 0.0266 0.1457 1>}.RGB
    Alpha=NamedRerouteUsage()
   )) {name=Colour, nodeColor=<rgba 0.5234 0 1 1>}
  n71 = ParticleColor()
  n72 = Panner(
   Coordinate=TextureCoordinate()
   Speed=AppendVector(A=ScalarParameter() {defaultValue=0.4, group=Pickaxe Swing Mask, parameterName="Swing Mask Panner Speed (U)", sortPriority=2}, B=Constant())
  ) {bFractionalPart=true}
  n73 = TextureCoordinate() {vTiling=0}
  n74 = Constant2Vector() {r=0.5}
  n75 = Multiply(
   A=ShadingPathSwitch(Default=Constant() {r=1}, Deferred=DepthFade())
   B=Multiply(
    A=TextureSample(UVs=@n72) {samplerType=SAMPLERTYPE_Grayscale, texture=&T_Smoke_Trail_Soft}.R
    B=SphereMask(A=@n73, B=@n74) {attenuationRadius=0.5, hardnessPercent=0}
   )
  )
  NamedRerouteDeclaration(None=LinearInterpolate(
    A=Multiply(A=@n71.A, B=@n75)
    B=NamedRerouteUsage()
    Alpha=ScalarParameter() {group=Pickaxe Swing Mask, parameterName=Use_simple_round_particle_mask, sortPriority=1}
   )) {name=Pickaxe_Swing_Mask, nodeColor=<rgba 0.5 0 1 1>}
  NamedRerouteDeclaration(None=MaterialFunctionCall(
    Value_Morning=ScalarParameter() {defaultValue=1, group=Time of Day, parameterName=Color Value Morning, sortPriority=0}
    Value_Day=ScalarParameter() {defaultValue=1, group=Time of Day, parameterName=Color Value Day, sortPriority=1}
    Value_Evening=ScalarParameter() {defaultValue=1, group=Time of Day, parameterName=Color Value Evening, sortPriority=2}
    Value_Night=ScalarParameter() {defaultValue=1, group=Time of Day, parameterName=Color Value Night, sortPriority=3}
   ) {materialFunction=&Scalar_Time_of_Day}.Result) {name=Time of Day Scale, nodeColor=<rgba 0 1 0.1797 1>}
  n76 = Add(A=ComponentMask(None=TextureCoordinate()) {g=true}) {constB=-0.72}
  NamedRerouteDeclaration(None=LinearInterpolate(Alpha=Saturate(None=@n76)) {constA=1, constB=7}) {name=Global Gradient, nodeColor=<rgba 1 0 0.1406 1>}
  n77 = DynamicParameter() {desc=Parameters set by Niagara, parameterIndex=1, paramNames [X,Y,Z,WPOIntensity]}
  n78 = Clamp(
   None=Subtract(
    A=MaterialFunctionCall(X=@n77.X, Y=@n77.X, Z=@n77.X) {materialFunction=&MakeFloat3}.Result
    B=ParticlePositionWS()
   )
   Min=ScalarParameter() {defaultValue=-100, group=Interpolate WPO, parameterName=Clamp_min, sortPriority=1}
   Max=ScalarParameter() {defaultValue=100, group=Interpolate WPO, parameterName=Clamp_max, sortPriority=2}
  ) {maxDefault=1000, minDefault=-1000}
  n79 = OneMinus(None=TextureCoordinate())
  n80 = Multiply(
   A=@n77.WPOIntensity
   B=Power(
    Base=ComponentMask(None=@n79) {g=true}
    Exp=ScalarParameter() {defaultValue=3, group=Interpolate WPO, parameterName=WPOBendPower, sortPriority=3}
   )
  )
  NamedRerouteDeclaration(None=LinearInterpolate(
    A=Constant()
    B=LinearInterpolate(B=@n78, Alpha=@n80)
    Alpha=ScalarParameter() {defaultValue=1, group=Interpolate WPO, parameterName=Use interpolate WPO, sortPriority=0}
   )) {name=Interpolate WPO, nodeColor=<rgba 0 0.4609 1 1>}
  n81 = TextureCoordinate() {coordinateIndex=1}
  n82 = TextureCoordinate()
  n83 = ScalarParameter() {parameterName=Use TexCoord 0 for All}
  n84 = Panner(
   Coordinate=Add(
    A=DynamicParameter() {paramNames [Offset, Texture_mask_offset, WaveIntensity, Param4]}.Offset
    B=LinearInterpolate(A=@n81, B=@n82, Alpha=@n83)
   )
   Time=Multiply(A=Time(), B=ScalarParameter() {defaultValue=1.1, group=WPO, parameterName="WPO Panning Speed Scale (0,0.8)", sortPriority=0})
  ) {bFractionalPart=true, speedY=0.8}
  n85 = TextureSample(UVs=@n84) {samplerType=SAMPLERTYPE_Normal, texture=&T_Weap_Distort_N_02}
  n86 = TextureCoordinate()
  n87 = LinearInterpolate(
   B=AppendVector(
    A=ComponentMask(None=@n85.RGB) {g=true, r=true}
    B=TextureSample(UVs=@n84) {texture=&T_Smooth_tile}.R
   )
   Alpha=OneMinus(None=ComponentMask(None=@n86) {g=true})
  )
  n88 = DynamicParameter() {paramNames [Offset, Texture_mask_offset, WaveIntensity, Param4]}
  NamedRerouteDeclaration(None=LinearInterpolate(A=Multiply(A=@n87, B=@n88.WaveIntensity) {constB=10}, Alpha=ScalarParameter() {desc=no_Wpo, group=WPO, parameterName=No_WPO}) {constB=0}) {name=WPO, nodeColor=<rgba 0.7812 1 0 1>}
  n89 = Add(A=DynamicParameter() {paramNames [Offset, Texture_mask_offset, WaveIntensity, Param4]}.Offset, B=TextureCoordinate() {vTiling=1.5})
  n90 = ComponentMask(None=VectorParameter() {group=Distortion, parameterName=Distortion UV Scale, sortPriority=3, defaultValue=<rgba 1 1 0 0>}.RGB) {g=true, r=true}
  n91 = Time()
  n92 = ScalarParameter() {defaultValue=1, group=Distortion, parameterName="Distortion Panner Speed Scale (0,1)", sortPriority=4}
  n93 = TextureSampleParameter2D(UVs=Panner(Coordinate=Multiply(A=@n89, B=@n90), Time=Multiply(A=@n91, B=@n92)) {bFractionalPart=true, speedY=1}) {
   group=Distortion
   parameterName=Distortion Normal
   samplerType=SAMPLERTYPE_Normal
   sortPriority=2
   texture=&T_Weap_Distort_N_02
  }
  NamedRerouteDeclaration(None=@n93.R) {name=Distortion Normal Red Channel, nodeColor=<rgba 0.8984 0 1 1>}
  n94 = ComponentMask(None=@n93.RGB) {g=true, r=true}
  n95 = VectorParameter() {group=Fire Shape, parameterName=Fire_MacroShape_UVScale, sortPriority=10, defaultValue=<rgba 1 1 0 0>}
  n96 = ComponentMask(None=Multiply(A=TextureCoordinate(), B=ComponentMask(None=@n95.RGB) {g=true, r=true})) {g=true}
  n97 = LinearInterpolate(
   B=Divide(
    A=LinearInterpolate(A=@n94, Alpha=@n96) {constB=0}
    B=ScalarParameter() {defaultValue=3, group=Distortion, parameterName=UV_Distortion_Divsor, sortPriority=1}
   ) {constB=20}
   Alpha=ScalarParameter() {defaultValue=1, group=Distortion, parameterName=UV_Distortion_Lerp, sortPriority=0}
  )
  n98 = Add(
   A=DynamicParameter() {paramNames [Offset, Texture_mask_offset, WaveIntensity, Param4]}.Offset
   B=LinearInterpolate(A=TextureCoordinate() {coordinateIndex=1}, B=TextureCoordinate(), Alpha=ScalarParameter() {parameterName=Use TexCoord 0 for All})
  )
  n99 = ComponentMask(None=VectorParameter() {group=Fire Shape, parameterName=Fire_Micro_UVScale, sortPriority=8, defaultValue=<rgba 1 1 0 0>}.RGB) {g=true, r=true}
  n100 = Panner(
   Coordinate=MaterialFunctionCall(XY=Multiply(A=@n98, B=@n99)) {materialFunction=&Swizzle}.YX
   Time=Multiply(
    A=Time()
    B=ScalarParameter() {defaultValue=1, group=Fire Shape, parameterName="Fire Micro Scale Panner Speed Scale (1.2,0.2)", sortPriority=9}
   )
  ) {bFractionalPart=true, speedX=1.2, speedY=0.2}
  n101 = TextureSampleParameter2D(UVs=ShadingPathSwitch(Default=Add(A=@n97, B=@n100), Mobile=@n100)) {
   group=Fire Shape
   parameterName=Fire Micro Texture
   samplerType=SAMPLERTYPE_LinearColor
   sortPriority=6
   texture=&T_NoisePack_LinearColor
  }
  n102 = TextureSample() {samplerType=SAMPLERTYPE_LinearGrayscale, texture=/Engine/Functions/Engine_MaterialFunctions02/ExampleContent/Textures/LowResBlurredNoise}
  NamedRerouteDeclaration(None=LinearInterpolate(
    A=ChannelMaskParameter(None=@n101.RGBA) {group=Fire Shape, parameterName=Fire Micro Mask, sortPriority=7}.R
    B=@n102.R
    Alpha=ScalarParameter() {defaultValue=0.2, group=Fire Shape, parameterName=Fire Micro lerp, sortPriority=5}
   ) {constAlpha=0.2}) {name=Fire Micro Mask, nodeColor=<rgba 0 1 0.5781 1>}
  NamedRerouteDeclaration(None=@n101.RGB) {name=Fire Micro Texture, nodeColor=<rgba 1 0.2578 0 1>}
 ]
}
